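# Physics asset for the DBNO (down but not out) variant of the M_MED_WingBath_Sharp male medium player body, covering only the lower torso and left leg.
# Capsules sit on pelvis, spine_01, thigh (two overlapping), calf and foot, plus a wide, thin root box just below the origin, shaped like a floor.
# Constraints chain the bodies back to root with all angular motion free, so the stored 45-degree limits are inactive.
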PhysicsAsset /BRCosmetics/Characters/Player/Male/Medium/Bodies/M_MED_WingBath_Sharp/Meshes/M_MED_WingBath_Sharp_DBNO_Physics
bodies {
 calf_l [{
   shapeType=Capsule
   length=32.064
   radius=7.0671
   shapeName=calf_l_capsule
   center=<-18.932 -4.2818 -0.0855>
   rotation=<rot -89.5921 -178.7809 177.2262>
  }]
 foot_l [{
   shapeType=Capsule
   length=10.5044
   radius=7.8707
   shapeName=foot_l_capsule
   center=<-2.2739 2.7461 -0.7256>
   rotation=<rot 0 -6.9877 -86.9664>
  }]
 pelvis [{shapeType=Capsule, length=4.5, radius=13.9, center=<-5.0556 0.5166 0>}]
 root [{shapeType=Box, extentX=200, extentY=200, extentZ=1, shapeName=root_box, center=<0 0 -1>, rotation=<rot 0 -90 0>}]
 spine_01 [{shapeType=Capsule, length=1.3, radius=14.1881, shapeName=spine_01_capsule, center=<5.1008 -1.0816 1.1749>}]
 thigh_l [
  {
   shapeType=Capsule
   length=30.3944
   radius=8.517
   shapeName=thigh_l_capsule
   center=<-20.8398 1.0337 0.5419>
   rotation=<rot 88.2691 -178.5999 179.9999>
  }
  {
   shapeType=Capsule
   length=37.3
   radius=7.6
   shapeName=thigh_l_capsule
   center=<-16.4904 -1.3978 -0.7527>
   rotation=<rot 89.8609 179.999 179.9991>
  }
 ]
}
constraints [
 {
  bone1Name=spine_01
  bone2Name=pelvis
  swing1LimitDegrees=45
  swing1Motion=Free
  swing2LimitDegrees=45
  swing2Motion=Free
  twistLimitDegrees=45
  twistMotion=Free
 }
 {
  bone1Name=thigh_l
  bone2Name=pelvis
  swing1LimitDegrees=45
  swing1Motion=Free
  swing2LimitDegrees=45
  swing2Motion=Free
  twistLimitDegrees=45
  twistMotion=Free
 }
 {
  bone1Name=calf_l
  bone2Name=thigh_l
  swing1LimitDegrees=45
  swing1Motion=Free
  swing2LimitDegrees=45
  swing2Motion=Free
  twistLimitDegrees=45
  twistMotion=Free
 }
 {
  bone1Name=foot_l
  bone2Name=calf_l
  swing1LimitDegrees=45
  swing1Motion=Free
  swing2LimitDegrees=45
  swing2Motion=Free
  twistLimitDegrees=45
  twistMotion=Free
 }
 {
  bone1Name=pelvis
  bone2Name=root
  swing1LimitDegrees=45
  swing1Motion=Free
  swing2LimitDegrees=45
  swing2Motion=Free
  twistLimitDegrees=45
  twistMotion=Free
 }
]
