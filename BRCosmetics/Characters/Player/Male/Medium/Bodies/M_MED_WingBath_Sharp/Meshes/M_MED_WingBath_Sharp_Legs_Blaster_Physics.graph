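# Physics asset for the legs blaster holster on the WingBath Sharp medium male player cosmetic body, built entirely from capsules.
# A holster chain hangs from the pelvis: a locked top-master anchor, a tightly limited driver, then the main holster with wider swing and twist limits.
# Pelvis, lower spine and right-leg capsules supply body volumes around the holster.
# Thigh and calf joints are locked, while foot and spine joints are free.
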PhysicsAsset /BRCosmetics/Characters/Player/Male/Medium/Bodies/M_MED_WingBath_Sharp/Meshes/M_MED_WingBath_Sharp_Legs_Blaster_Physics
bodies {
 calf_r [
  {shapeType=Capsule, length=21.2247, radius=7.8595, center=<11.0572 2.1068 -0.751>, rotation=<rot -0.0765 94.2096 90.6596>}
  {shapeType=Capsule, length=16.2539, radius=7.2423, shapeName=calf_r_capsule, center=<31.9769 0.8513 0>, rotation=<rot -0.0002 79.9977 90>}
 ]
 dyn_holster_driver [{
   shapeType=Capsule
   length=5.79
   radius=2.0474
   shapeName=dyn_bagDrive_01_capsule
   center=<5.032 0 -0.0023>
   rotation=<rot 0.0009 -89.9997 90.1755>
  }]
 dyn_holster_main [
  {shapeType=Capsule, length=25.6658, radius=4.1486, shapeName=dyn_bag_00_capsule, center=<1.4976 1.2147 4.2089>, rotation=<rot 0 91.6526 90>}
  {
   shapeType=Capsule
   length=7.5
   radius=4.5
   shapeName=dyn_holster_main_capsule
   center=<-10.3371 3.5718 5.5582>
   rotation=<rot -1.3541 36.0985 88.9833>
  }
 ]
 dyn_holster_top_master [{shapeType=Capsule, length=0.505, radius=0.505, shapeName=dyn_holster_top_master_capsule, rotation=<rot 0 -90 0>}]
 foot_r [{
   shapeType=Capsule
   length=10.4457
   radius=7.7736
   shapeName=foot_r_capsule
   center=<2.9019 -5.2824 0.869>
   rotation=<rot -0.3918 2.2746 -86.6329>
  }]
 pelvis [
  {shapeType=Capsule, length=9.6555, radius=10.5103, center=<2.5806 -3.0576 0>}
  {shapeType=Capsule, length=8.674, radius=3.9016, shapeName=pelvis_capsule, center=<10.2603 5.6214 13.0573>, rotation=<rot 0 -0.006 -55.7307>}
 ]
 spine_01 [{shapeType=Capsule, length=7.1831, radius=12.9111, center=<4.9494 0.4609 0>}]
 thigh_r [
  {shapeType=Capsule, length=31.3134, radius=10.3294, center=<18.0702 0.0003 -3.742>, rotation=<rot -90 180 -2.7375>}
  {shapeType=Capsule, length=7.7896, radius=7.9604, center=<17.5102 -0.5933 -1.1246>, rotation=<rot -90 90 89.9999>}
 ]
}
constraints [
 {
  bone1Name=thigh_r
  bone2Name=pelvis
  swing1LimitDegrees=45
  swing1Motion=Locked
  swing2LimitDegrees=45
  swing2Motion=Locked
  twistLimitDegrees=45
  twistMotion=Locked
 }
 {
  bone1Name=dyn_holster_top_master
  bone2Name=pelvis
  swing1LimitDegrees=45
  swing1Motion=Locked
  swing2LimitDegrees=45
  swing2Motion=Locked
  twistLimitDegrees=45
  twistMotion=Locked
 }
 {bone1Name=dyn_holster_driver, bone2Name=dyn_holster_top_master, swing1LimitDegrees=5, swing2LimitDegrees=10, twistLimitDegrees=5}
 {bone1Name=dyn_holster_main, bone2Name=dyn_holster_driver, swing1LimitDegrees=18, swing2LimitDegrees=30, twistLimitDegrees=15}
 {
  bone1Name=calf_r
  bone2Name=thigh_r
  swing1LimitDegrees=45
  swing1Motion=Locked
  swing2LimitDegrees=45
  swing2Motion=Locked
  twistLimitDegrees=45
  twistMotion=Locked
 }
 {
  bone1Name=foot_r
  bone2Name=calf_r
  swing1LimitDegrees=45
  swing1Motion=Free
  swing2LimitDegrees=45
  swing2Motion=Free
  twistLimitDegrees=45
  twistMotion=Free
 }
 {
  bone1Name=spine_01
  bone2Name=pelvis
  swing1LimitDegrees=45
  swing1Motion=Free
  swing2LimitDegrees=45
  swing2Motion=Free
  twistLimitDegrees=45
  twistMotion=Free
 }
]
