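# Physics asset for the legs mesh of the M_MED_WingBath_Sharp player body, named for the DBNO (downed) state, covering left-side skirt dynamics.
# Front, side and back chains each hang three limited-angle capsules from a tiny base locked to the pelvis.
# Pelvis, spine_01 and left leg capsules use free constraints, providing body shapes around the skirt.
# A wide, thin box on root sits just below the origin like a floor.
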PhysicsAsset /BRCosmetics/Characters/Player/Male/Medium/Bodies/M_MED_WingBath_Sharp/Meshes/M_MED_WingBath_Sharp_Legs_DBNO_Physics
bodies {
 calf_l [{
   shapeType=Capsule
   length=32.064
   radius=7.0671
   shapeName=calf_l_capsule
   center=<-18.932 -4.2818 -0.0855>
   rotation=<rot -89.5921 -178.7809 177.2262>
  }]
 dyn_skirt_back_01_l [{shapeType=Capsule, length=6.8, radius=4.2, shapeName=dyn_skirt_back_02_l_capsule, center=<6.5752 -2.4171 0>, rotation=<rot 90 0 0>}]
 dyn_skirt_back_02_l [{
   shapeType=Capsule
   length=4.4
   radius=5.3
   shapeName=dyn_skirt_back_02_l_capsule
   center=<7.3748 -3.0244 1.7809>
   rotation=<rot 90 -90 -84.5387>
  }]
 dyn_skirt_back_03_l [{
   shapeType=Capsule
   length=6.4
   radius=5.1
   shapeName=dyn_skirt_back_03_l_capsule
   center=<9.8003 -4.1219 1.0802>
   rotation=<rot 90 180 -172.6304>
  }]
 dyn_skirt_back_Base_l [{shapeType=Capsule, length=0.505, radius=0.505, shapeName=dyn_skirt_back_Base_l_capsule, rotation=<rot 0 -90 0>}]
 dyn_skirt_front_01_l [{
   shapeType=Capsule
   length=6.8
   radius=4.2
   shapeName=dyn_skirt_back_02_l_capsule
   center=<5.8575 -3.0967 0.9499>
   rotation=<rot 58.9829 95.9669 97.5128>
  }]
 dyn_skirt_front_02_l [{shapeType=Capsule, length=6.5, radius=4.2, shapeName=dyn_skirt_back_02_l_capsule, center=<6.811 -3.8239 2.5525>, rotation=<rot 90 0 0>}]
 dyn_skirt_front_03_l [{shapeType=Capsule, length=9.8, radius=5, shapeName=dyn_skirt_front_03_l_capsule, center=<10.7139 -2.6715 1.7215>, rotation=<rot 90 0 0>}]
 dyn_skirt_front_Base_l [{shapeType=Capsule, length=0.505, radius=0.505, shapeName=dyn_skirt_front_Base_l_capsule, rotation=<rot 0 -90 0>}]
 dyn_skirt_side_01_l [{shapeType=Capsule, length=6.8, radius=4.2, shapeName=dyn_skirt_back_02_l_capsule, center=<7.3985 -2.0375 0>, rotation=<rot 90 0 0>}]
 dyn_skirt_side_02_l [{shapeType=Capsule, length=6.8, radius=4.2, shapeName=dyn_skirt_back_02_l_capsule, center=<7.3985 -2.7764 0>, rotation=<rot 90 0 0>}]
 dyn_skirt_side_03_l [{shapeType=Capsule, length=2.9, radius=4.6, shapeName=dyn_skirt_side_03_l_capsule, center=<8.5076 -2.7897 0>, rotation=<rot 90 0 0>}]
 dyn_skirt_side_Base_l [{shapeType=Capsule, length=0.505, radius=0.505, shapeName=dyn_skirt_side_Base_l_capsule, rotation=<rot 0 -90 0>}]
 foot_l [{
   shapeType=Capsule
   length=10.5044
   radius=7.8707
   shapeName=foot_l_capsule
   center=<-2.2739 2.7461 -0.7256>
   rotation=<rot 0 -6.9877 -86.9664>
  }]
 pelvis [{shapeType=Capsule, length=4.5, radius=13.9, center=<-5.0556 0.5166 0>}]
 root [{shapeType=Box, extentX=200, extentY=200, extentZ=1, shapeName=root_box, center=<0 0 -1>, rotation=<rot 0 -90 0>}]
 spine_01 [{shapeType=Capsule, length=1.3, radius=14.1881, shapeName=spine_01_capsule, center=<5.1008 -1.0816 1.1749>}]
 thigh_l [
  {
   shapeType=Capsule
   length=30.3944
   radius=8.517
   shapeName=thigh_l_capsule
   center=<-20.8398 1.0337 0.5419>
   rotation=<rot 88.2691 -178.5999 179.9999>
  }
  {
   shapeType=Capsule
   length=37.3
   radius=7.6
   shapeName=thigh_l_capsule
   center=<-16.4904 -1.3978 -0.7527>
   rotation=<rot 89.8609 179.999 179.9991>
  }
 ]
}
constraints [
 {
  bone1Name=pelvis
  bone2Name=root
  swing1LimitDegrees=45
  swing1Motion=Free
  swing2LimitDegrees=45
  swing2Motion=Free
  twistLimitDegrees=45
  twistMotion=Free
 }
 {
  bone1Name=dyn_skirt_back_Base_l
  bone2Name=pelvis
  swing1LimitDegrees=45
  swing1Motion=Locked
  swing2LimitDegrees=45
  swing2Motion=Locked
  twistLimitDegrees=45
  twistMotion=Locked
 }
 {
  bone1Name=dyn_skirt_front_Base_l
  bone2Name=pelvis
  swing1LimitDegrees=45
  swing1Motion=Locked
  swing2LimitDegrees=45
  swing2Motion=Locked
  twistLimitDegrees=45
  twistMotion=Locked
 }
 {
  bone1Name=dyn_skirt_side_Base_l
  bone2Name=pelvis
  swing1LimitDegrees=45
  swing1Motion=Locked
  swing2LimitDegrees=45
  swing2Motion=Locked
  twistLimitDegrees=45
  twistMotion=Locked
 }
 {
  bone1Name=spine_01
  bone2Name=pelvis
  swing1LimitDegrees=45
  swing1Motion=Free
  swing2LimitDegrees=45
  swing2Motion=Free
  twistLimitDegrees=45
  twistMotion=Free
 }
 {
  bone1Name=thigh_l
  bone2Name=pelvis
  swing1LimitDegrees=45
  swing1Motion=Free
  swing2LimitDegrees=45
  swing2Motion=Free
  twistLimitDegrees=45
  twistMotion=Free
 }
 {bone1Name=dyn_skirt_back_01_l, bone2Name=dyn_skirt_back_Base_l, swing1LimitDegrees=10, swing2LimitDegrees=0, twistLimitDegrees=0}
 {bone1Name=dyn_skirt_back_02_l, bone2Name=dyn_skirt_back_01_l, swing1LimitDegrees=45, swing2LimitDegrees=10, twistLimitDegrees=5}
 {bone1Name=dyn_skirt_back_03_l, bone2Name=dyn_skirt_back_02_l, swing1LimitDegrees=10, swing2LimitDegrees=10, twistLimitDegrees=25}
 {bone1Name=dyn_skirt_front_01_l, bone2Name=dyn_skirt_front_Base_l, swing1LimitDegrees=60, swing2LimitDegrees=22, twistLimitDegrees=5}
 {bone1Name=dyn_skirt_front_02_l, bone2Name=dyn_skirt_front_01_l, swing1LimitDegrees=10, swing2LimitDegrees=10, twistLimitDegrees=60}
 {bone1Name=dyn_skirt_front_03_l, bone2Name=dyn_skirt_front_02_l, swing1LimitDegrees=10, swing2LimitDegrees=10, twistLimitDegrees=40}
 {bone1Name=dyn_skirt_side_01_l, bone2Name=dyn_skirt_side_Base_l, swing1LimitDegrees=30, swing2LimitDegrees=10, twistLimitDegrees=0}
 {bone1Name=dyn_skirt_side_02_l, bone2Name=dyn_skirt_side_01_l, swing1LimitDegrees=40, swing2LimitDegrees=10, twistLimitDegrees=60}
 {bone1Name=dyn_skirt_side_03_l, bone2Name=dyn_skirt_side_02_l, swing1LimitDegrees=22, swing2LimitDegrees=22, twistLimitDegrees=35}
 {
  bone1Name=calf_l
  bone2Name=thigh_l
  swing1LimitDegrees=45
  swing1Motion=Free
  swing2LimitDegrees=45
  swing2Motion=Free
  twistLimitDegrees=45
  twistMotion=Free
 }
 {
  bone1Name=foot_l
  bone2Name=calf_l
  swing1LimitDegrees=45
  swing1Motion=Free
  swing2LimitDegrees=45
  swing2Motion=Free
  twistLimitDegrees=45
  twistMotion=Free
 }
]
