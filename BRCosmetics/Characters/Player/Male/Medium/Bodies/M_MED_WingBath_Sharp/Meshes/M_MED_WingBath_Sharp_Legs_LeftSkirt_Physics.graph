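# Physics asset for the left skirt on the legs of the male medium WingBath Sharp player cosmetic body.
# Front, side and back chains of three box bodies hang from zero-size base bodies locked angularly to the pelvis, with angular limits along each chain.
# Bottom segments are cross-linked, back to side with free rotation and side to front with limits.
# Pelvis, left leg and foot capsules plus a wide, thin root box provide collision shapes.
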PhysicsAsset /BRCosmetics/Characters/Player/Male/Medium/Bodies/M_MED_WingBath_Sharp/Meshes/M_MED_WingBath_Sharp_Legs_LeftSkirt_Physics
bodies {
 calf_l [{shapeType=Capsule, length=28.5772, radius=11.6844, center=<-22.6943 0 0.9072>, rotation=<rot 90 0 0>}]
 dyn_skirt_back_01_l [{shapeType=Box, extentX=14, extentY=2.9452, extentZ=10.1258, center=<6.6408 -0.3288 -0.1369>, rotation=<rot 0.0001 0.66 -22.5963>}]
 dyn_skirt_back_02_l [{shapeType=Box, extentX=12, extentY=3.7764, extentZ=11.8877, center=<8.2749 -0.1697 -0.0758>, rotation=<rot 0.0053 0.0034 -24.0775>}]
 dyn_skirt_back_03_l [{shapeType=Box, extentX=17, extentY=3.674, extentZ=12.1853, center=<9.0558 -0.4485 -0.6253>, rotation=<rot -2.0568 0.9156 -19.4466>}]
 dyn_skirt_back_Base_l [{shapeType=Capsule, length=0, radius=0, shapeName=dyn_skirt_back_Base_l_capsule}]
 dyn_skirt_front_01_l [{
   shapeType=Box
   extentX=11.5459
   extentY=3.0311
   extentZ=12.9589
   shapeName=dyn_skirt_front_01_l_box
   center=<8.9997 -1.8256 0.6646>
   rotation=<rot 3.8709 -0.0056 -0.0009>
  }]
 dyn_skirt_front_02_l [{
   shapeType=Box
   extentX=14.2384
   extentY=4.7034
   extentZ=17.2769
   shapeName=dyn_skirt_front_02_l_box
   center=<8.8944 -1.9607 1.132>
   rotation=<rot 2.1264 3.6854 11.6658>
  }]
 dyn_skirt_front_03_l [{
   shapeType=Box
   extentX=15.6046
   extentY=2.9809
   extentZ=18.3086
   shapeName=dyn_skirt_front_03_l_box
   center=<8.0448 -0.2149 1.6169>
   rotation=<rot 5.455 5.8096 9.0988>
  }]
 dyn_skirt_front_Base_l [{shapeType=Capsule, length=0, radius=0, shapeName=dyn_skirt_front_Base_l_capsule}]
 dyn_skirt_side_01_l [{shapeType=Box, extentX=15, extentY=2.1815, extentZ=8, center=<7.5553 0 0>, rotation=<rot -0.4723 0.0003 -0.0009>}]
 dyn_skirt_side_02_l [{shapeType=Box, extentX=16.4384, extentY=4.1034, extentZ=12.5247, center=<7.8495 0 0>}]
 dyn_skirt_side_03_l [{shapeType=Box, extentX=15, extentY=4.7986, extentZ=13.5665, center=<7.5059 0 0>}]
 dyn_skirt_side_Base_l [{shapeType=Capsule, length=0, radius=0, shapeName=dyn_skirt_side_Base_l_capsule}]
 foot_l [{shapeType=Capsule, length=12.7957, radius=8.1395, center=<-4.6409 5.1059 0>, rotation=<rot 90 180 87.7526>}]
 pelvis [{shapeType=Capsule, length=9.6354, radius=11.7125, center=<-4 0 0>}, {shapeType=Capsule, length=5.7775, radius=11.7125, center=<6.0372 0 0>}]
 root [{shapeType=Box, extentX=200, extentY=200, extentZ=1.01, shapeName=root_box}]
 thigh_l [
  {shapeType=Capsule, length=41.3401, radius=9.4847, center=<-16.6762 -1.6815 0.2507>, rotation=<rot 89.6107 173.7125 171.2134>}
  {
   shapeType=Capsule
   length=21.5
   radius=7.5
   shapeName=thigh_l_capsule
   center=<-16.7404 -0.0349 -0.7149>
   rotation=<rot 87.5157 179.9987 177.2977>
  }
  {shapeType=Capsule, length=38.4216, radius=7.9118, center=<-21.339 -3.6314 0.3923>, rotation=<rot 88.29 179.9997 179.9997>}
 ]
}
constraints [
 {
  bone1Name=dyn_skirt_back_Base_l
  bone2Name=pelvis
  swing1LimitDegrees=10
  swing1Motion=Locked
  swing2LimitDegrees=10
  swing2Motion=Locked
  twistLimitDegrees=10
  twistMotion=Locked
 }
 {
  bone1Name=dyn_skirt_front_Base_l
  bone2Name=pelvis
  swing1LimitDegrees=10
  swing1Motion=Locked
  swing2LimitDegrees=10
  swing2Motion=Locked
  twistLimitDegrees=10
  twistMotion=Locked
 }
 {
  bone1Name=dyn_skirt_side_Base_l
  bone2Name=pelvis
  swing1LimitDegrees=10
  swing1Motion=Locked
  swing2LimitDegrees=10
  swing2Motion=Locked
  twistLimitDegrees=10
  twistMotion=Locked
 }
 {bone1Name=dyn_skirt_back_01_l, bone2Name=dyn_skirt_back_Base_l, swing1LimitDegrees=15, swing2LimitDegrees=45, twistLimitDegrees=15}
 {bone1Name=dyn_skirt_back_02_l, bone2Name=dyn_skirt_back_01_l, swing1LimitDegrees=50, swing2LimitDegrees=50, twistLimitDegrees=30}
 {bone1Name=dyn_skirt_back_03_l, bone2Name=dyn_skirt_back_02_l, swing1LimitDegrees=40, swing2LimitDegrees=60, twistLimitDegrees=30}
 {
  bone1Name=dyn_skirt_side_03_l
  bone2Name=dyn_skirt_back_03_l
  swing1LimitDegrees=45
  swing1Motion=Free
  swing2LimitDegrees=45
  swing2Motion=Free
  twistLimitDegrees=45
  twistMotion=Free
 }
 {bone1Name=dyn_skirt_front_01_l, bone2Name=dyn_skirt_front_Base_l, swing1LimitDegrees=40, swing2LimitDegrees=30, twistLimitDegrees=60}
 {bone1Name=dyn_skirt_front_02_l, bone2Name=dyn_skirt_front_01_l, swing1LimitDegrees=40, swing2LimitDegrees=40, twistLimitDegrees=60}
 {bone1Name=dyn_skirt_front_03_l, bone2Name=dyn_skirt_front_02_l, swing1LimitDegrees=45, swing2LimitDegrees=60, twistLimitDegrees=30}
 {bone1Name=dyn_skirt_front_03_l, bone2Name=dyn_skirt_side_03_l, swing1LimitDegrees=45, swing2LimitDegrees=45, twistLimitDegrees=45}
 {bone1Name=dyn_skirt_side_01_l, bone2Name=dyn_skirt_side_Base_l, swing1LimitDegrees=10, swing2LimitDegrees=15, twistLimitDegrees=55}
 {bone1Name=dyn_skirt_side_02_l, bone2Name=dyn_skirt_side_01_l, swing1LimitDegrees=50, swing2LimitDegrees=50, twistLimitDegrees=30}
 {bone1Name=dyn_skirt_side_03_l, bone2Name=dyn_skirt_side_02_l, swing1LimitDegrees=45, swing2LimitDegrees=60, twistLimitDegrees=30}
]
