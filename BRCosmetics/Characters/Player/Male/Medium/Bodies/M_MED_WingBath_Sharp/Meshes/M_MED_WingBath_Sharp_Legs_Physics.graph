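# Physics asset for the legs mesh of the M_MED_WingBath_Sharp male player body cosmetic, providing capsule-only collision bodies.
# Capsules cover pelvis, lower spine, right thigh, calf and foot, plus a very large one on root; the left leg has none.
# A two-bone dyn_Blaster_back chain hangs from the pelvis with narrow angular limits, apparently for secondary motion.
# Thigh and calf joints are angularly locked, while the foot rotates freely.
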PhysicsAsset /BRCosmetics/Characters/Player/Male/Medium/Bodies/M_MED_WingBath_Sharp/Meshes/M_MED_WingBath_Sharp_Legs_Physics
bodies {
 calf_r [
  {shapeType=Capsule, length=21.2247, radius=7.8595, center=<11.0572 2.1068 -0.751>, rotation=<rot -0.0765 94.2096 90.6596>}
  {shapeType=Capsule, length=16.2539, radius=7.2423, shapeName=calf_r_capsule, center=<31.9769 0.8513 0>, rotation=<rot -0.0002 79.9977 90>}
 ]
 dyn_Blaster_back_02 [{
   shapeType=Capsule
   length=3
   radius=2.5
   shapeName=dyn_Blaster_back_02_capsule
   center=<4.2493 0.1963 -1.6964>
   rotation=<rot 81.5626 -58.8572 -62.3231>
  }]
 dyn_Blaster_back_03 [{
   shapeType=Capsule
   length=3
   radius=2.5
   shapeName=dyn_Blaster_back_02_capsule
   center=<4.2269 0.5954 -1.644>
   rotation=<rot 73.2449 -150.9873 -153.2088>
  }]
 foot_r [{
   shapeType=Capsule
   length=10.4457
   radius=7.7736
   shapeName=foot_r_capsule
   center=<2.9019 -5.2824 0.869>
   rotation=<rot -0.3918 2.2746 -86.6329>
  }]
 pelvis [{shapeType=Capsule, length=12.2957, radius=10.5103, center=<-4.2548 -3.0576 0>}]
 root [{shapeType=Capsule, length=129, radius=151.5, shapeName=root_capsule, center=<0 -18.2808 -152.4107>, rotation=<rot 0 90 90>}]
 spine_01 [{shapeType=Capsule, length=0, radius=14.75, center=<7.1334 -0.0002 0>}]
 thigh_r [
  {shapeType=Capsule, length=30.3309, radius=11.3118, center=<18.0702 0.0003 -2.8263>, rotation=<rot -90 -90.3524 -92.3851>}
  {shapeType=Capsule, length=6.25, radius=9.5, center=<17.5102 -0.5933 0.3279>, rotation=<rot -90 0.0099 179.99>}
 ]
}
constraints [
 {
  bone1Name=thigh_r
  bone2Name=pelvis
  swing1LimitDegrees=45
  swing1Motion=Locked
  swing2LimitDegrees=45
  swing2Motion=Locked
  twistLimitDegrees=45
  twistMotion=Locked
 }
 {bone1Name=dyn_Blaster_back_02, bone2Name=pelvis, swing1LimitDegrees=20, swing2LimitDegrees=20, twistLimitDegrees=15}
 {bone1Name=dyn_Blaster_back_03, bone2Name=dyn_Blaster_back_02, swing1LimitDegrees=20, swing2LimitDegrees=20, twistLimitDegrees=15}
 {
  bone1Name=calf_r
  bone2Name=thigh_r
  swing1LimitDegrees=45
  swing1Motion=Locked
  swing2LimitDegrees=45
  swing2Motion=Locked
  twistLimitDegrees=45
  twistMotion=Locked
 }
 {
  bone1Name=foot_r
  bone2Name=calf_r
  swing1LimitDegrees=45
  swing1Motion=Free
  swing2LimitDegrees=45
  swing2Motion=Free
  twistLimitDegrees=45
  twistMotion=Free
 }
]
