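# Physics asset for the left skirt mesh of the M_MED_WingBath_Sharp male body cosmetic, providing simple collision shapes.
# Single capsules cover the pelvis, left calf and left foot, while three overlapping, offset capsules fill out the left thigh.
# A wide, thin box on the root bone forms a floor-like slab.
# It defines no constraints, so the bodies form an unjointed collision set rather than a ragdoll.
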PhysicsAsset /BRCosmetics/Characters/Player/Male/Medium/Bodies/M_MED_WingBath_Sharp/Meshes/M_MED_WingBath_Sharp_Skirt_left_Physics
bodies {
 calf_l [{shapeType=Capsule, length=32.6849, radius=11.6844, center=<-20.2002 0 0.9072>, rotation=<rot 90 0 0>}]
 foot_l [{shapeType=Capsule, length=12.7957, radius=8.1395, center=<-4.6409 5.1059 0>, rotation=<rot 90 180 87.7526>}]
 pelvis [{shapeType=Capsule, length=9.6354, radius=11.7125, center=<-4 0 0>}]
 root [{shapeType=Box, extentX=200, extentY=200, extentZ=1.01, shapeName=root_box}]
 thigh_l [
  {shapeType=Capsule, length=38.1983, radius=8.135, center=<-20.2067 0 0.2164>, rotation=<rot 88.29 179.959 179.2471>}
  {
   shapeType=Capsule
   length=21.5
   radius=7.5
   shapeName=thigh_l_capsule
   center=<-16.92 3.4244 -0.7227>
   rotation=<rot 87.5157 179.9992 170.3706>
  }
  {shapeType=Capsule, length=38.1983, radius=8.135, center=<-20.2067 -3.1934 0.2164>, rotation=<rot 88.29 179.9997 179.9997>}
 ]
}
constraints []
